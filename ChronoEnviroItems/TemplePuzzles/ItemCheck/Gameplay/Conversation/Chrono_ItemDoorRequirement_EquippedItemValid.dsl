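; Controller requirement for the item door puzzle conversation that tests whether the player's equipped item counts as a valid item.
; It casts the test subject to a Fort player controller, treats any held object as invalid, and resolves the current weapon's inventory item by GUID.
; It passes only if that item has no tag from the Not Valid Item container and is neither a harvesting weapon nor a building piece.
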(fn IsRequirementMetInternal
  (params (RequestContext :type Controller%20Requirement%20Test%20Context%20Structure%20%28by%20ref%29))
  (bind (_testsubjectactor _testsubjectcontroller) (Utilities|Struct|BreakControllerRequirementTestContext RequestContext))
  (bind _asfort_player_controller (Utilities|Casting|CastToFortPlayerController _testsubjectcontroller)
    (:then
      (Utilities|IsValid _asfort_player_controller
        (:Is%20Valid
          ;[1 1 1] If held object, treat as invalid item
          (bind _returnvalue (Fortnite|HeldObjects|GetHeldObjectAttachedtoPlayer (Game|Player|GetPlayerPawn _asfort_player_controller)))
          (Utilities|IsValid _returnvalue
            (:Is%20Valid
              _)
            (:Is%20Not%20Valid
              (Utilities|IsValid (Game|Player|GetPlayerPawn _asfort_player_controller)
                (:Is%20Valid
                  (Utilities|IsValid (Weapon|GetCurrentWeapon (Game|Player|GetPlayerPawn _asfort_player_controller))
                    (:Is%20Valid
                      (if (Guid|IsValidGuid (FortWeapon|GetInventoryGUID (Weapon|GetCurrentWeapon (Game|Player|GetPlayerPawn _asfort_player_controller))))
                        (Utilities|IsValid (Inventory|GetInventoryItemWithGuid _asfort_player_controller (FortWeapon|GetInventoryGUID (Weapon|GetCurrentWeapon (Game|Player|GetPlayerPawn _asfort_player_controller))))
                          (:Is%20Valid
                            ;[1 1 1] FORT-621547 - Equipped or held item is valid? Covers cases like augment device.
                            (bind _returnvalue_2 (GameplayTags|HasAnyMatchingGameplayTags(Message) (Inventory|GetInventoryItemWithGuid _asfort_player_controller (FortWeapon|GetInventoryGUID (Weapon|GetCurrentWeapon (Game|Player|GetPlayerPawn _asfort_player_controller)))) (Variables|Default|GetNot%20Valid%20Item%20Tag%20Container)))
                            (return (and (not _returnvalue_2) (not (or (Utilities|Enum|Equal(Enum) (Item|GetType (Inventory|GetInventoryItemWithGuid _asfort_player_controller (FortWeapon|GetInventoryGUID (Weapon|GetCurrentWeapon (Game|Player|GetPlayerPawn _asfort_player_controller))))) "WeaponHarvest") (Utilities|Enum|Equal(Enum) (Item|GetType (Inventory|GetInventoryItemWithGuid _asfort_player_controller (FortWeapon|GetInventoryGUID (Weapon|GetCurrentWeapon (Game|Player|GetPlayerPawn _asfort_player_controller))))) "BuildingPiece"))))))
                          (:Is%20Not%20Valid
                            _))
                        (else
                          _))
                      (return false))
                    (:Is%20Not%20Valid
                      _)))
                (:Is%20Not%20Valid
                  _)))))
        (:Is%20Not%20Valid
          _)))
    (:CastFailed
      _)))
