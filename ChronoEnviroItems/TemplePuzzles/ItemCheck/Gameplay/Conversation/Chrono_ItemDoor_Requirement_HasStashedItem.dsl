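; Conversation requirement for the temple item-check puzzle door, named to test whether the player has a stashed item.
; It overrides IsRequirementMetInternal, which takes a controller requirement test context by reference.
; The function returns the value of a variable named Temp directly, without reading the context or inspecting any inventory, so the check appears to be a placeholder.
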(fn IsRequirementMetInternal
  (params (RequestContext :type Controller%20Requirement%20Test%20Context%20Structure%20%28by%20ref%29))
  (return (Variables|Default|GetTemp)))
