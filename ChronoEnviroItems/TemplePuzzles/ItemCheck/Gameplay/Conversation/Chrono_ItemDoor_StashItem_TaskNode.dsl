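; Conversation task node for a Chrono temple puzzle door that demands the player's held item meet a required rarity.
; Qualifying items are consumed, firing the door-open event, cue and tags; lower-rarity items are taken, tossed back as pickups, and advance the conversation.
; Mythic or Transcendent items also send a stash quest stat event, before the door check.
; GatherStaticExtraData passes the held item's identity and stack or ammo count as conversation parameters.
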(fn ExecuteTaskNode
  (params (Context :type Conversation%20Context%20Structure%20%28by%20ref%29))
  (bind _context (Variables|GetContext))
  (Development|PrintString "STASH ITEM Task Node Start")
  (bind _returnvalue (Class|FortConversationContextLibrary|GetParticipantActorNPC Context))
  (if (not (Variables|/Script/FortniteGame.BuildingActor|GetbDestroyed (Utilities|Casting|CastToBuildingActor _returnvalue)))
    ;[1 1 1] Retrieve stashed item from from player component, and clear persisted data
    (bind _returnvalue_2 (Class|FortConversationContextLibrary|GetParticipantActorPlayer Context))
    (Utilities|IsValid _returnvalue_2
      (:Is%20Valid
        (bind _meetsraritycriteria (CallFunction|ItemRarityMythicorTranscendent _returnvalue_2))
        (if _meetsraritycriteria
          (CallFunction|SendQuestEventforRarity _returnvalue_2)
          (else
            _))
        (bind _returnvalue_3 (Class|FortConversationContextLibrary|GetParticipantActorNPC (Variables|GetContext)))
        ;[0.3096 0 1] Check against the doors required rarity to open. True == consume item and open, False == reject item
        (bind _asbp_chrono_requires_item_temple_door (Utilities|Casting|CastToBP_Chrono_RequiresItem_TempleDoor _returnvalue_3))
        (if (>= (Math|Byte|MakeLiteralByte (Rarity|GetWeaponDataRarity (Weapon|GetCurrentWeapon (Game|Player|GetPlayerPawn (Fortnite|GetFortPlayerControllerfromActor _returnvalue_2))))) (Math|Byte|MakeLiteralByte (Variables|/ChronoEnviroItems/TemplePuzzles/ItemCheck/Gameplay/Actors/BP_Chrono_RequiresItem_TempleDoor.BP_Chrono_RequiresItem_TempleDoor_C|GetRarity%20Check%20Required%20Rarity _asbp_chrono_requires_item_temple_door)))
          ;[0.1831 1 0.1051] If we give an appropriate item,  remove Item from inventory & send event to open the door
          (Utilities|IsValid (Weapon|GetCurrentWeapon (Game|Player|GetPlayerPawn (Fortnite|GetFortPlayerControllerfromActor _returnvalue_2)))
            (:Is%20Valid
              (bind _asfort_inventory_owner_interface (Utilities|Casting|CastToFortInventoryOwnerInterface (Fortnite|GetFortPlayerControllerfromActor _returnvalue_2))
                (:then
                  (InventoryManagement|RemoveItem _asfort_inventory_owner_interface (FortWeapon|GetInventoryGUID (Weapon|GetCurrentWeapon (Game|Player|GetPlayerPawn (Fortnite|GetFortPlayerControllerfromActor _returnvalue_2)))) 1)
                  (bind _returnvalue_4 (Class|FortConversationContextLibrary|GetParticipantActorNPC (Variables|GetContext)))
                  (Ability|SendGameplayEventtoActor _returnvalue_4 "(TagName=\"Event.TemplePuzzle.ItemDoor.Enabled\")")
                  (GameplayCue|ExecuteGameplayCue _returnvalue_4 (Variables|Default|GetGC%20Accepted%20Item))
                  (Ability|GameplayEffect|AddLooseGameplayTags _returnvalue_4 (Variables|Default|GetDoorOpenGameplay%20Tags))
                  (bind _returnvalue_5 (|CallParentFunction (Variables|GetContext)))
                  (return _returnvalue_5))
                (:CastFailed
                  _)))
            (:Is%20Not%20Valid
              _))
          (bind _returnvalue_6 (Conversation|AbortConversation (Variables|GetContext)))
          (return _returnvalue_6)
          (else
            ;[0.02924 1 0.9403] Item is too low rarity, we remove the item from the player's inventory and 'spit' it back out into the world, then advaance conversation to give rejected message
            (bind _returnvalue_7 (Class|FortConversationContextLibrary|GetParticipantActorPlayer _context))
            (bind _returnvalue_8 (Class|FortConversationContextLibrary|GetParticipantActorNPC _context))
            (bind _asbp_chrono_requires_item_temple_door_2 (Utilities|Casting|CastToBP_Chrono_RequiresItem_TempleDoor _returnvalue_8)
              (:then
                (bind _asfort_inventory_owner_interface_2 (Utilities|Casting|CastToFortInventoryOwnerInterface (Fortnite|GetFortPlayerControllerfromActor _returnvalue_7))
                  (:then
                    (Utilities|IsValid (Variables|/ChronoEnviroItems/TemplePuzzles/ItemCheck/Gameplay/Actors/BP_Chrono_RequiresItem_TempleDoor.BP_Chrono_RequiresItem_TempleDoor_C|GetFortItemEntry _asbp_chrono_requires_item_temple_door_2)
                      (:Is%20Valid
                        (Class|FortItemEntryComponent|TakeItemfromInventoryOwner (Variables|/ChronoEnviroItems/TemplePuzzles/ItemCheck/Gameplay/Actors/BP_Chrono_RequiresItem_TempleDoor.BP_Chrono_RequiresItem_TempleDoor_C|GetFortItemEntry _asbp_chrono_requires_item_temple_door_2) (FortWeapon|GetInventoryGUID (Weapon|GetCurrentWeapon (Game|Player|GetPlayerPawn (Fortnite|GetFortPlayerControllerfromActor _returnvalue_7)))) -1 _asfort_inventory_owner_interface_2 true) ; -1 removes a stack of items correctly based on count
                        (bind _returnvalue_9 (Inventory|SpawnPickupInWorldWithClassAndItemEntry (Class|FortItemEntryComponent|GetOwnedItem (Variables|/ChronoEnviroItems/TemplePuzzles/ItemCheck/Gameplay/Actors/BP_Chrono_RequiresItem_TempleDoor.BP_Chrono_RequiresItem_TempleDoor_C|GetFortItemEntry _asbp_chrono_requires_item_temple_door_2)) "/Script/FortniteGame.FortPickupAthena" (CallFunction|GetSpawnItemPosition _returnvalue_8) (CallFunction|GetDirectionandTossEndPoint _returnvalue_8) 0 true true false "Tossed" "Unset" 0 false))
                        (Pickup|TossPickup _returnvalue_9 (CallFunction|GetDirectionandTossEndPoint _returnvalue_8) 0 0 true true "Tossed" "Unset")
                        (bind _returnvalue_10 (Class|FortConversationContextLibrary|GetParticipantActorNPC (Variables|GetContext)))
                        (GameplayCue|ExecuteGameplayCue _returnvalue_10 (Variables|Default|GetGC%20Refused%20Item))
                        (bind _returnvalue_11 (Conversation|AdvanceConversation (Variables|GetContext)))
                        (return _returnvalue_11))
                      (:Is%20Not%20Valid
                        _)))
                  (:CastFailed
                    _)))
              (:CastFailed
                _))
            (Development|PrintString "Cast Failed for Item Door")
            (bind _returnvalue_12 (Conversation|AbortConversation (Variables|GetContext)))
            (return _returnvalue_12)))))
    (else
      (Development|PrintString "error: tent is destroyed, unable to stash")
      (bind _returnvalue_13 (Conversation|AbortConversation (Variables|GetContext)))
      (return _returnvalue_13))))

(fn GatherStaticExtraData
  (params (Context :type Conversation%20Context%20Structure%20%28by%20ref%29))
  (bind _inoutextradata (|CallParentFunction Context))
  (bind _returnvalue (Class|FortConversationContextLibrary|GetParticipantActorPlayer Context))
  (Utilities|IsValid _returnvalue
    (:Is%20Valid
      (Utilities|IsValid (Fortnite|GetFortPlayerControllerfromActor _returnvalue)
        (:Is%20Valid
          (Utilities|IsValid (Game|Player|GetPlayerPawn (Fortnite|GetFortPlayerControllerfromActor _returnvalue))
            (:Is%20Valid
              (Utilities|IsValid (Weapon|GetCurrentWeapon (Game|Player|GetPlayerPawn (Fortnite|GetFortPlayerControllerfromActor _returnvalue)))
                (:Is%20Valid
                  (if (Guid|IsValidGuid (FortWeapon|GetInventoryGUID (Weapon|GetCurrentWeapon (Game|Player|GetPlayerPawn (Fortnite|GetFortPlayerControllerfromActor _returnvalue)))))
                    (Utilities|IsValid (Inventory|GetInventoryItemWithGuid (Fortnite|GetFortPlayerControllerfromActor _returnvalue) (FortWeapon|GetInventoryGUID (Weapon|GetCurrentWeapon (Game|Player|GetPlayerPawn (Fortnite|GetFortPlayerControllerfromActor _returnvalue)))))
                      (:Is%20Valid
                        ;[1 1 1] Ensure parent params sent in result, in addition to any custom added here
                        (Utilities|Array|AppendArray _inoutextradata (Utilities|Array|MakeArray (Utilities|Struct|MakeConversationNodeParameterPair "displayinventoryitem" (Guid|ToString(Guid) (FortWeapon|GetInventoryGUID (Weapon|GetCurrentWeapon (Game|Player|GetPlayerPawn (Fortnite|GetFortPlayerControllerfromActor _returnvalue)))))) (Utilities|Struct|MakeConversationNodeParameterPair "displayitem" (Utilities|ToString(SoftObjectReference) (Utilities|ResolveSoftReference (Inventory|GetInventoryItemWithGuid (Fortnite|GetFortPlayerControllerfromActor _returnvalue) (FortWeapon|GetInventoryGUID (Weapon|GetCurrentWeapon (Game|Player|GetPlayerPawn (Fortnite|GetFortPlayerControllerfromActor _returnvalue)))))))) (Utilities|Struct|MakeConversationNodeParameterPair "stacksize" (select (and (>= (Item|GetRemainingAmmo (Inventory|GetInventoryItemWithGuid (Fortnite|GetFortPlayerControllerfromActor _returnvalue) (FortWeapon|GetInventoryGUID (Weapon|GetCurrentWeapon (Game|Player|GetPlayerPawn (Fortnite|GetFortPlayerControllerfromActor _returnvalue)))))) 0) (Weapon|DoesItemConsumeUniqueRangedWeaponAmmo (Inventory|GetInventoryItemWithGuid (Fortnite|GetFortPlayerControllerfromActor _returnvalue) (FortWeapon|GetInventoryGUID (Weapon|GetCurrentWeapon (Game|Player|GetPlayerPawn (Fortnite|GetFortPlayerControllerfromActor _returnvalue))))))) (Utilities|String|ToString(Integer) (Item|GetRemainingAmmo (Inventory|GetInventoryItemWithGuid (Fortnite|GetFortPlayerControllerfromActor _returnvalue) (FortWeapon|GetInventoryGUID (Weapon|GetCurrentWeapon (Game|Player|GetPlayerPawn (Fortnite|GetFortPlayerControllerfromActor _returnvalue))))))) (Utilities|String|ToString(Integer) (Item|GetNuminStack (Inventory|GetInventoryItemWithGuid (Fortnite|GetFortPlayerControllerfromActor _returnvalue) (FortWeapon|GetInventoryGUID (Weapon|GetCurrentWeapon (Game|Player|GetPlayerPawn (Fortnite|GetFortPlayerControllerfromActor _returnvalue))))))))) (Utilities|Struct|MakeConversationNodeParameterPair "ShowWarningMessage" (Utilities|String|SelectString "true" "" (CallFunction|WasPreviouslyStashedItemInvalid _returnvalue))))))
                      (:Is%20Not%20Valid
                        (Fortnite|FortShippingLog self "GatherStaticExtraData: Inventory Item with GUID not valid (GetInventoryItemWithGUID)" true)))
                    (return _inoutextradata)
                    (else
                      (Fortnite|FortShippingLog self "GatherStaticExtraData: Inventory GUID not valid" true))))
                (:Is%20Not%20Valid
                  (Fortnite|FortShippingLog self "GatherStaticExtraData: Could not get Current Weapon" true))))
            (:Is%20Not%20Valid
              (Fortnite|FortShippingLog self "GatherStaticExtraData: Could not get Player Pawn" true))))
        (:Is%20Not%20Valid
          (Fortnite|FortShippingLog self "GatherStaticExtraData: Could not get Player Controller from Actor" true))))
    (:Is%20Not%20Valid
      (Fortnite|FortShippingLog self "GatherStaticExtraData: Could not get Participan Actor Player" true))))

(fn GetDirectionAndTossEndPoint
  (params (self3 :type Actor%20Object%20Reference))
  ;[1 1 1] offset up past bowl on door and away from door which is right vector. Random the distance a little
  (return (+ (+ (Transformation|GetActorLocation self3) "0.000000,0.000000,180.000000") (* (Transformation|GetActorRightVector self3) (Math|Random|RandomFloatinRange (LocalVariables|GetMinThrowDist) (LocalVariables|GetMaxThrowDist))))))

(fn GetSpawnItemPosition
  (params (self3 :type Actor%20Object%20Reference))
  ;[1 1 1] offset spawn item above bowl on door and forward from door a bit
  (return (+ (+ (Transformation|GetActorLocation self3) "0.000000,0.000000,100.000000") (* (Transformation|GetActorRightVector self3) 1e+02))))

(fn ItemRarityMythicOrTranscendent
  (params (Player :type Actor%20Object%20Reference))
  (bind _returnvalue (GameplayTags|HasAnyMatchingGameplayTags(Message) (Inventory|GetInventoryItemWithGuid (Fortnite|GetFortPlayerControllerfromActor Player) (FortWeapon|GetInventoryGUID (Weapon|GetCurrentWeapon (Game|Player|GetPlayerPawn (Fortnite|GetFortPlayerControllerfromActor Player))))) (LocalVariables|GetRarityTags)))
  (return _returnvalue))

(fn ItemRarityRareOrAbove
  (params (Player :type Actor%20Object%20Reference))
  (bind _returnvalue (GameplayTags|HasAnyMatchingGameplayTags(Message) (Inventory|GetInventoryItemWithGuid (Fortnite|GetFortPlayerControllerfromActor Player) (FortWeapon|GetInventoryGUID (Weapon|GetCurrentWeapon (Game|Player|GetPlayerPawn (Fortnite|GetFortPlayerControllerfromActor Player))))) (LocalVariables|GetRarityTags)))
  (return _returnvalue))

(fn SendQuestEventForRarity
  (params (Player :type Actor%20Object%20Reference))
  (Quest|SendCustomStatEventwithTags (Quests|GetQuestManager (Fortnite|GetFortPlayerControllerfromActor Player) "Invalid") "UsedNPCService" "(GameplayTags=((TagName=\"Athena.Quests.Stash.Item.HighRarity\")))" "(GameplayTags=)" 1)
  (return))

(fn WasPreviouslyStashedItemInvalid
  (params (Actor :type Actor%20Object%20Reference))
  (Utilities|IsValid 0
    (:Is%20Valid
      (if (not true)
        (return false)))
    (:Is%20Not%20Valid
      (return false))))
