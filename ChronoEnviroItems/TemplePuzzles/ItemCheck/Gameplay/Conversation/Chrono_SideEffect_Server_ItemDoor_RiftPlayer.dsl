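; Server-side conversation side effect for the Chrono item door puzzle that rifts the participating player away.
; If the NPC participant building is intact, it spawns a rift portal at the player pawn's mesh location and forces that pawn to teleport.
; It then plays a gameplay cue on the NPC and calls the parent; a destroyed NPC aborts the conversation.
; Debug strings and the cue variable still refer to stashing items in a tent.
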(fn ServerCauseSideEffect
  (params (Context :type Conversation%20Context%20Structure%20%28by%20ref%29))
  (Development|PrintString "STASH ITEM Task Node Start")
  (bind _returnvalue (Class|FortConversationContextLibrary|GetParticipantActorNPC Context))
  (if (not (Variables|/Script/FortniteGame.BuildingActor|GetbDestroyed (Utilities|Casting|CastToBuildingActor _returnvalue)))
    ;[1 1 1] spawn rift on player
    (bind _returnvalue_2 (Class|FortConversationContextLibrary|GetParticipantActorPlayer Context))
    (Utilities|IsValid _returnvalue_2
      (:Is%20Valid
        (bind _asfort_player_pawn_athena (Utilities|Casting|CastToFortPlayerPawnAthena (Game|Player|GetPlayerPawn (Fortnite|GetFortPlayerControllerfromActor _returnvalue_2))))
        (bind _returnvalue_3 (Game|SpawnActorfromClass :Class "/Game/Athena/Items/Consumables/RiftItem/BGA_RiftPortal_Item_Athena.BGA_RiftPortal_Item_Athena_C" :SpawnTransform (Math|Transform|MakeTransform (Transformation|GetWorldLocation (Variables|/Script/Engine.Character|GetMesh _asfort_player_pawn_athena))) :CollisionHandlingOverride "AdjustIfPossibleButAlwaysSpawn" :TransformScaleMethod "OverrideRootScale"))
        (Class|BGARiftPortalItemAthena|ForceRiftTeleport _returnvalue_3 (Game|Player|GetPlayerPawn (Fortnite|GetFortPlayerControllerfromActor _returnvalue_2)))
        (GameplayCue|ExecuteGameplayCue _returnvalue (Variables|Default|GetGC%20StashItem))
        (|CallParentFunction Context)
        (return)))
    (else
      (Development|PrintString "error: tent is destroyed, unable to stash")
      (Conversation|AbortConversation Context)
      (return))))
