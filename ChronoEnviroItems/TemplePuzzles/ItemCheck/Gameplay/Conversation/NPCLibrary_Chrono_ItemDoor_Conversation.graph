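# Conversation database for the Chrono temple puzzle item door, which demands a worthy item to pass and warns that given items are consumed.
# A GIVE ITEM choice leads to BACK or OFFER ITEM, a stash task gated by stashed-item, curve-table hotfix and equipped-item requirements.
# The closing speech defaults to a low-quality rejection, switching to acceptance when the NPC owns the door Opened tag.
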ConversationDatabase /ChronoEnviroItems/TemplePuzzles/ItemCheck/Gameplay/Conversation/NPCLibrary_Chrono_ItemDoor_Conversation
entry_points [{tag=Conversation.Entry.RequiresItemDoor, destinations [#1]}]
nodes [
 ConversationEntryPointNode_7 {
  id=#1
  class=ConversationEntryPointNode
  outputs [#2]
  properties {entryTag {tagName=Conversation.Entry.RequiresItemDoor}}
 }
 Speech_C_4 {
  id=#2
  class=Speech_C
  outputs [#3]
  properties {nodeName="", generalConfig {defaultMessage="Give me a worthy item to pass!", defaultMetadataParameters [Style {value=Mad}]}}
 }
 Speech_C_7 {
  id=#3
  class=Speech_C
  outputs [#4, #5]
  properties {generalConfig {defaultMessage=Items you give me will be consumed., defaultMetadataParameters [Style {value=""}]}}
  sub_nodes [FortConversationChoiceWithPreviewNode_4 {class=FortConversationChoiceWithPreviewNode, properties {defaultChoiceDisplayText=GIVE ITEM}}]
 }
 Chrono_ItemDoor_StashItem_TaskNode_C_3 {
  id=#5
  class=Chrono_ItemDoor_StashItem_TaskNode_C
  outputs [#6]
  properties {requirements [
    {
     cannotUseReasonParameter=FullStash
     failureNodeBehaviour=FailedAndHidden
     participantId {tagName=Conversation.Participant.Player}
     requirement {class=Chrono_ItemDoor_Requirement_HasStashedItem_C}
    }
    {
     cannotUseReasonParameter=ServiceDisabled
     failureNodeBehaviour=FailedButVisible
     participantId {tagName=Conversation.Participant.Player}
     requirement {
      class=Chrono_ItemDoor_Requirement_HotfixEnabled_C
      properties {isServiceEnabled? {value=1, curve {curveTable=/ChronoEnviroItems/DataTables/ChronoEnviroItemsGameData, rowName=Chrono.TemplePuzzle.ItemDoor.Enabled}}}
     }
    }
    {
     cannotUseReasonParameter=UnstashableItem
     failureNodeBehaviour=FailedButVisible
     participantId {tagName=Conversation.Participant.Player}
     requirement {class=Chrono_ItemDoorRequirement_EquippedItemValid_C}
    }
   ]}
  sub_nodes [ConversationChoiceNode_9 {
    class=ConversationChoiceNode
    properties {
     defaultChoiceDisplayText=OFFER ITEM
     choiceTags {gameplayTags [{tagName=Conversation.Choice.InitialSelectedIndex}, {tagName=Conversation.Choice.SetFocusUsingCurrentIndex}]}
    }
   }]
 }
 Speech_C_0 {
  id=#6
  class=Speech_C
  outputs []
  properties {generalConfig {
    defaultMessage="This item is too low quality. Bring me another!"
    contextualMessages [{
      message=This is acceptable. You may pass.
      requirementMatchPolicy=RequireAny
      contextRequirements [{
        participantId {tagName=Conversation.Participant.NPC}
        requirement {class=FortControllerRequirement_HasOwnedTag, properties {ownedTag {tagName=Chrono.EnviroItems.RuinsPuzzles.ItemCheckDoor.Opened}}}
       }]
      priority {value=0, curve {curveTable=null, rowName=None}}
     }]
   }}
 }
 Back_C_0 {
  id=#4
  class=Back_C
  sub_nodes [ConversationChoiceNode_2 {class=ConversationChoiceNode, properties {defaultChoiceDisplayText=BACK}}]
 }
]
