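# NPC conversation database for the Chrono temple pattern door puzzle, giving spoken hints about finding and matching flames.
# The Conversation.Entry.PatternDoor entry point leads into a linear chain of two Speech nodes, with no choices or branches.
# The first line talks of finding flames and carries a Mad style parameter, and the second tells of matching them, ending the conversation.
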ConversationDatabase /ChronoEnviroItems/TemplePuzzles/ItemCheck/Gameplay/Conversation/NPCLibrary_Chrono_PatternDoor_Conversation
entry_points [{tag=Conversation.Entry.PatternDoor, destinations [#1]}]
nodes [
 ConversationEntryPointNode_7 {
  id=#1
  class=ConversationEntryPointNode
  outputs [#2]
  properties {entryTag {tagName=Conversation.Entry.PatternDoor}}
 }
 Speech_C_1 {
  id=#2
  class=Speech_C
  outputs [#3]
  properties {nodeName="", generalConfig {defaultMessage=... find ... flames, defaultMetadataParameters [Style {value=Mad}]}}
 }
 Speech_C_7 {id=#3, class=Speech_C, outputs [], properties {generalConfig {defaultMessage=Match ... flames ...}}}
]
