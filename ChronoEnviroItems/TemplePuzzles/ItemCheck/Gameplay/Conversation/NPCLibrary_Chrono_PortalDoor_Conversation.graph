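# Conversation database for the portal statue door NPC in the Chrono temple item-check puzzle.
# A single entry point, tagged Conversation.Entry.PortalStatueDoor, leads into a linear chain of two speech nodes with no choices or conditions.
# The first line, in a Mad style, tells the player to seek a statue, and the second follows with a fragmentary hint before the conversation ends.
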ConversationDatabase /ChronoEnviroItems/TemplePuzzles/ItemCheck/Gameplay/Conversation/NPCLibrary_Chrono_PortalDoor_Conversation
entry_points [{tag=Conversation.Entry.PortalStatueDoor, destinations [#1]}]
nodes [
 ConversationEntryPointNode_7 {
  id=#1
  class=ConversationEntryPointNode
  outputs [#2]
  properties {entryTag {tagName=Conversation.Entry.PortalStatueDoor}}
 }
 Speech_C_1 {
  id=#2
  class=Speech_C
  outputs [#3]
  properties {nodeName="", generalConfig {defaultMessage=Seek ... statue, defaultMetadataParameters [Style {value=Mad}]}}
 }
 Speech_C_7 {id=#3, class=Speech_C, outputs [], properties {generalConfig {defaultMessage="... is the ...  one"}}}
]
